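# Unlit translucent sprite material for Niagara and particle systems that renders the Strawberry Pilot's strawberry emote effect.
# A channel-packed texture drives it: blue indexes a pink gradient and a curve-atlas row offset by a dynamic parameter, while green masks both.
# These are lighten-blended, red adds a cyan tint, and particle color feeds an HDR-derived emissive adjustment.
# A boosted, saturated red channel times particle alpha forms a sharpened mask.
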Material /Game/Characters/Player/Female/Medium/Bodies/F_MED_Strawberry_Pilot/Textures/FX/M_KPopV4_Emote_Strawberry
graph {
 properties {
  bUsedWithNiagaraSprites=true
  bUsedWithParticleSprites=true
  blendMode=BLEND_Translucent
  pixelDepthOffsetMode=PDOM_Legacy
  shadingModel=MSM_Unlit
  translucencyLightingMode=TLM_Surface
  usedShadingModels=Unlit
 }
 nodes [
  n1 = TextureSample() {texture=&/T_KPopV4_Emote_Strawberry}
  n2 = Add(A=OneMinus(None=@n1.B), B=ScalarParameter() {defaultValue=-0.1})
  n3 = LinearInterpolate(
   B=TextureSample(UVs=Clamp(None=@n2)) {texture=/Game/Characters/Player/Female/Medium/Bodies/F_MED_Strawberry_Pilot/Textures/Gradients/Default/T_Pink_2}.RGB
   Alpha=@n1.G
  )
  n4 = CurveAtlasRowParameter(CurveTime=Add(A=@n1.B, B=DynamicParameter().Param1)) {
   atlas=&AthenaCurveAtlas01
   curve=&/Curve_Strawberry
   defaultValue=97
   parameterName=Param_1
  }
  n5 = Multiply(A=Power(Base=@n4) {constExponent=1.6}) {constB=1.3}
  n6 = Add(
   A=MaterialFunctionCall!(Blend=Multiply(A=@n3) {constB=0.9}, Base=LinearInterpolate(B=@n5, Alpha=@n1.G)) {materialFunction=/Engine/Functions/Engine_MaterialFunctions03/Blends/Blend_Lighten}.Result
   B=LinearInterpolate(B=Constant3Vector() {constant=<rgba 0 0.8023 1 0>}, Alpha=@n1.R)
  )
  Add(B=@n6)
  n7 = Multiply(A=@n6) {constB=0.8}
  n8 = ParticleColor()
  MaterialFunctionCall(Emissive=MaterialFunctionCall(LDR Input=Saturate(None=@n7), HDR Tint=Multiply(A=@n8.RGB, B=@n7)) {materialFunction=&DeriveHDRfromLDR}.RemappedRange) {materialFunction=&MF_FX_EmissiveAdjustments}
  n9 = Add(A=@n1.R, B=@n1.R)
  Multiply(A=@n8.A, B=Saturate(None=Multiply(A=@n9) {constB=20}))
 ]
}
